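; Athena vending machine Blueprint variant "Rufus" for the BlastBerry map, which stays in the match only when its curve table toggle is on.
; On BeginPlay it waits for playlist data, then evaluates the Enabled Row curve value.
; A positive value runs the parent's BeginPlay; otherwise it hides the quest vending machine mesh and destroys the actor.
; Its marker details ID returns None, and the overlap, tick and markable events are empty.
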(event EventBeginPlay
  (Class|FortAsyncActionCurrentPlaylistReady|PlaylistDataReadyAsync
    (:OnReady
      (bind _outxy (Fortnite|EvaluateCurveTableRow (Variables|Default|GetEnabled%20Row)))
      (if (> _outxy 0)
        (|CallParentFunction)
        (else
          (Rendering|SetVisibility (Variables|Default|GetSM_VendingMachineQuest) false false)
          (Actor|DestroyActor))))))

(event Collision|EventActorBeginOverlap
  (params (OtherActor :type Actor%20Object%20Reference)))

(event EventTick
  (params (DeltaSeconds :type Float%20%28single-precision%29)))

(event MarkableActor|EventSetCanBeMarked
  (params (bInCanBeMarked :type Boolean)))

(fn GetMarkerDetailsId
  (return "None"))

(fn ConstructionScript
  (|CallParentFunction))
